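# Animation Blueprint for the female medium RoseJewel cosmetic body, adding secondary motion to a pose copied from the attached parent mesh.
# A crouch-aware Control Rig feeds weighted bone constraints for fur and hand spikes, plus spring AnimDynamics for fur and foot belts, cached as ClothingDynamics.
# BodyDynamics extends that cache with offset upper-torso bones and stacked angular and linear spring simulations.
# bPbBlendEnabled outputs BodyDynamics when true, otherwise ClothingDynamics.
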AnimBlueprint /BRCosmetics/Characters/Player/Female/Medium/Bodies/F_MED_RoseJewel/Meshes/F_MED_RoseJewel_AnimBP
target_skeleton=&/F_MED_RoseJewel_Skeleton
graphs [
 AnimGraph {nodes [
   Output Pose(Result=Blend Poses by bool(
     BlendPose_0=Use cached pose 'BodyDynamics'()
     BlendPose_1=Use cached pose 'ClothingDynamics'()
     bActiveValue=Get bPbBlendEnabled() -> (bPbBlendEnabled:bool)
    )) {layerGroup=None}
   n1 = Control Rig!(Source=Copy Pose From Mesh() {bUseAttachedParent=true}, Alpha, isCrouching:bool=Get bIsCrouching()) -> (Pose) {controlRigAssetReference {blueprintRigClass=&/F_MED_RoseJewel_CtrlRig.F_MED_RoseJewel_CtrlRig_C}}
   n2 = "Constraint - dyn_fur_side_r (PPPPPPPP)"!(
    ComponentPose=Local To Component(LocalPose=@n1)
    Alpha=1
    ConstraintWeights_0=0.0017
    ConstraintWeights_1:float=0.0648
    ConstraintWeights_2:float=0.6254
    ConstraintWeights_3:float=0.0982
    ConstraintWeights_4:float=0.0003
    ConstraintWeights_5:float=0.0559
    ConstraintWeights_6:float=0.0122
    ConstraintWeights_7:float=0.1415
   ) -> (Pose) {
    boneToModify {boneName=dyn_fur_side_r}
    constraintSetup [
     {targetBone {boneName=spine_04}}
     {targetBone {boneName=spine_05}}
     {targetBone {boneName=clavicle_r}}
     {targetBone {boneName=upperarm_twist_01_r}}
     {targetBone {boneName=deform_pec_r}}
     {targetBone {boneName=deform_upperarm_bk_r}}
     {targetBone {boneName=deform_upperarm_fr_r}}
     {targetBone {boneName=deform_upperarm_up_r}}
    ]
    constraintWeights [1,1,1,1,1,1,1,1]
   }
   n3 = "Constraint - dyn_fur_rear_r (PPPPPPPP)"!(
    ComponentPose=@n2
    Alpha=1
    ConstraintWeights_0=0.0269
    ConstraintWeights_1:float=0.3627
    ConstraintWeights_2:float=0.005
    ConstraintWeights_3:float=0.5623
    ConstraintWeights_4:float=0.0072
    ConstraintWeights_5:float=0.0046
    ConstraintWeights_6:float=0.0208
    ConstraintWeights_7:float=0.0106
   ) -> (Pose) {
    boneToModify {boneName=dyn_fur_rear_r}
    constraintSetup [
     {targetBone {boneName=spine_04}}
     {targetBone {boneName=spine_05}}
     {targetBone {boneName=clavicle_l}}
     {targetBone {boneName=clavicle_r}}
     {targetBone {boneName=upperarm_twist_01_r}}
     {targetBone {boneName=neck_01}}
     {targetBone {boneName=deform_upperarm_bk_r}}
     {targetBone {boneName=deform_upperarm_up_r}}
    ]
    constraintWeights [1,1,1,1,1,1,1,1]
   }
   n4 = "Constraint - dyn_fur (PPPPPPPP)"!(
    ComponentPose=@n3
    Alpha=1
    ConstraintWeights_0=0.0031
    ConstraintWeights_1:float=0.7887
    ConstraintWeights_2:float=0.1035
    ConstraintWeights_3:float=0.1035
    ConstraintWeights_4:float=0.0009
    ConstraintWeights_5:float=0.0002
    ConstraintWeights_6:float=0
    ConstraintWeights_7:float=0
   ) -> (Pose) {
    boneToModify {boneName=dyn_fur}
    constraintSetup [
     {targetBone {boneName=spine_04}}
     {targetBone {boneName=spine_05}}
     {targetBone {boneName=clavicle_l}}
     {targetBone {boneName=clavicle_r}}
     {targetBone {boneName=neck_01}}
     {targetBone {boneName=neck_02}}
     {targetBone {boneName=deform_upperarm_bk_l}}
     {targetBone {boneName=deform_upperarm_bk_r}}
    ]
    constraintWeights [1,1,1,1,1,1,1,1]
   }
   n5 = "Constraint - dyn_fur_rear_l (PPPPPPPP)"!(
    ComponentPose=@n4
    Alpha=1
    ConstraintWeights_0=0.0072
    ConstraintWeights_1:float=0.0269
    ConstraintWeights_2:float=0.3627
    ConstraintWeights_3:float=0.5623
    ConstraintWeights_4:float=0.005
    ConstraintWeights_5:float=0.0046
    ConstraintWeights_6:float=0.0208
    ConstraintWeights_7:float=0.0106
   ) -> (Pose) {
    boneToModify {boneName=dyn_fur_rear_l}
    constraintSetup [
     {targetBone {boneName=upperarm_twist_01_l}}
     {targetBone {boneName=spine_04}}
     {targetBone {boneName=spine_05}}
     {targetBone {boneName=clavicle_l}}
     {targetBone {boneName=clavicle_r}}
     {targetBone {boneName=neck_01}}
     {targetBone {boneName=deform_upperarm_bk_l}}
     {targetBone {boneName=deform_upperarm_up_l}}
    ]
    constraintWeights [1,1,1,1,1,1,1,1]
   }
   n6 = "Constraint - dyn_fur_side_l (PPPPPPPP)"!(
    ComponentPose=@n5
    Alpha=1
    ConstraintWeights_0=0.0982
    ConstraintWeights_1:float=0.0017
    ConstraintWeights_2:float=0.0648
    ConstraintWeights_3:float=0.6254
    ConstraintWeights_4:float=0.0003
    ConstraintWeights_5:float=0.0559
    ConstraintWeights_6:float=0.0122
    ConstraintWeights_7:float=0.1415
   ) -> (Pose) {
    boneToModify {boneName=dyn_fur_side_l}
    constraintSetup [
     {targetBone {boneName=upperarm_twist_01_l}}
     {targetBone {boneName=spine_04}}
     {targetBone {boneName=spine_05}}
     {targetBone {boneName=clavicle_l}}
     {targetBone {boneName=deform_pec_l}}
     {targetBone {boneName=deform_upperarm_bk_l}}
     {targetBone {boneName=deform_upperarm_fr_l}}
     {targetBone {boneName=deform_upperarm_up_l}}
    ]
    constraintWeights [1,1,1,1,1,1,1,1]
   }
   n7 = Anim Dynamics - Bone: dyn_fur(ComponentPose=@n6, Alpha=1, ExternalForce="0,0,0") {
    angularDampingOverride=0.9999
    angularSpringConstant=600
    bAngularSpring=true
    bOverrideAngularDamping=true
    bUseGravityOverride=true
    simulationSpace=RootRelative
    boundBone {boneName=dyn_fur}
    componentAppliedLinearAccClamp=<2000 2000 2000>
    componentLinearAccScale=<0.2 0.2 0.2>
    componentLinearVelScale=<0.2 0.2 0.5>
    gravityOverride {z=-980}
    physicsBodyDefinitions [{
      boundBone {boneName=dyn_fur}
      boxExtents=<13 13 13>
      constraintSetup {angularLimitsMax=<0 10 10>, angularLimitsMin=<0 -10 -10>, angularTarget=<1 0 0>}
      localJointOffset=<0.5 0 0>
     }]
    relativeSpaceBone {boneName=spine_05}
    simSpaceSettings {maxAngularAcceleration=1500, maxAngularVelocity=1500, simSpaceAngularAlpha=0.1}
   }
   n8 = Anim Dynamics - Bone: dyn_fur_rear_l(ComponentPose=@n7, Alpha=1, ExternalForce="0,0,0") {
    angularDampingOverride=0.9999
    angularSpringConstant=600
    bAngularSpring=true
    bOverrideAngularDamping=true
    bUseGravityOverride=true
    simulationSpace=RootRelative
    boundBone {boneName=dyn_fur_rear_l}
    componentAppliedLinearAccClamp=<2000 2000 2000>
    componentLinearAccScale=<0.2 0.2 0.2>
    componentLinearVelScale=<0.2 0.2 0.5>
    gravityOverride {z=-980}
    physicsBodyDefinitions [{
      boundBone {boneName=dyn_fur_rear_l}
      boxExtents=<12 12 12>
      constraintSetup {angularLimitsMax=<0 0 10>, angularLimitsMin=<0 -10 -10>, angularTarget=<1 0 0>}
      localJointOffset=<0.55 0 0>
     }]
    relativeSpaceBone {boneName=spine_05}
    simSpaceSettings {maxAngularAcceleration=1500, maxAngularVelocity=1500, simSpaceAngularAlpha=0.1}
   }
   n9 = Anim Dynamics - Bone: dyn_fur_rear_r(ComponentPose=@n8, Alpha=1, ExternalForce="0,0,0") {
    angularDampingOverride=0.9999
    angularSpringConstant=600
    bAngularSpring=true
    bOverrideAngularDamping=true
    bUseGravityOverride=true
    simulationSpace=RootRelative
    boundBone {boneName=dyn_fur_rear_r}
    componentAppliedLinearAccClamp=<2000 2000 2000>
    componentLinearAccScale=<0.2 0.2 0.2>
    componentLinearVelScale=<0.2 0.2 0.5>
    gravityOverride {z=-980}
    physicsBodyDefinitions [{
      boundBone {boneName=dyn_fur_rear_r}
      boxExtents=<11.5 11.5 11.5>
      constraintSetup {angularLimitsMax=<0 0 10>, angularLimitsMin=<0 -10 -10>, angularTarget=<1 0 0>}
      localJointOffset=<0.5 0 0>
     }]
    relativeSpaceBone {boneName=spine_05}
    simSpaceSettings {maxAngularAcceleration=1500, maxAngularVelocity=1500, simSpaceAngularAlpha=0.1}
   }
   n10 = Anim Dynamics - Bone: dyn_fur_side_l(ComponentPose=@n9, Alpha=1, ExternalForce="0,0,0") {
    angularDampingOverride=0.9999
    angularSpringConstant=600
    bAngularSpring=true
    bOverrideAngularDamping=true
    bUseGravityOverride=true
    simulationSpace=RootRelative
    boundBone {boneName=dyn_fur_side_l}
    componentAppliedLinearAccClamp=<2000 2000 2000>
    componentLinearAccScale=<0.2 0.2 0.2>
    componentLinearVelScale=<0.2 0.2 0.5>
    gravityOverride {z=-980}
    physicsBodyDefinitions [{
      boundBone {boneName=dyn_fur_side_l}
      boxExtents=<12.5 12.5 12.5>
      constraintSetup {angularLimitsMax=<0 0 10>, angularLimitsMin=<0 0 -10>, angularTarget=<1 0 0>}
      localJointOffset=<0.5 0 0>
     }]
    relativeSpaceBone {boneName=spine_05}
    simSpaceSettings {maxAngularAcceleration=1500, maxAngularVelocity=1500, simSpaceAngularAlpha=0.1}
   }
   n11 = Anim Dynamics - Bone: dyn_fur_side_r(ComponentPose=@n10, Alpha=1, ExternalForce="0,0,0") {
    angularDampingOverride=0.9999
    angularSpringConstant=600
    bAngularSpring=true
    bOverrideAngularDamping=true
    bUseGravityOverride=true
    simulationSpace=RootRelative
    boundBone {boneName=dyn_fur_side_r}
    componentAppliedLinearAccClamp=<2000 2000 2000>
    componentLinearAccScale=<0.2 0.2 0.2>
    componentLinearVelScale=<0.2 0.2 0.5>
    gravityOverride {z=-980}
    physicsBodyDefinitions [{
      boundBone {boneName=dyn_fur_side_r}
      boxExtents=<13 13 13>
      constraintSetup {angularLimitsMax=<0 0 10>, angularLimitsMin=<0 0 -10>, angularTarget=<1 0 0>}
      localJointOffset=<0.5 0 0>
     }]
    relativeSpaceBone {boneName=spine_05}
    simSpaceSettings {maxAngularAcceleration=1500, maxAngularVelocity=1500, simSpaceAngularAlpha=0.1}
   }
   n12 = "Constraint - def_spike04_l (PPPPPPP)"!(
    ComponentPose=@n11
    Alpha=1
    ConstraintWeights_0=0.5876
    ConstraintWeights_1:float=0.092
    ConstraintWeights_2:float=0.0023
    ConstraintWeights_3:float=0.0003
    ConstraintWeights_4:float=0.0506
    ConstraintWeights_5:float=0.2001
    ConstraintWeights_6:float=0.0671
   ) -> (Pose) {
    boneToModify {boneName=def_spike04_l}
    constraintSetup [
     {targetBone {boneName=hand_l}}
     {targetBone {boneName=ring_metacarpal_l}}
     {targetBone {boneName=middle_metacarpal_l}}
     {targetBone {boneName=index_metacarpal_l}}
     {targetBone {boneName=pinky_metacarpal_l}}
     {targetBone {boneName=lowerarm_twist_01_l}}
     {targetBone {boneName=deform_wrist_up_l}}
    ]
    constraintWeights [1,1,1,1,1,1,1]
   }
   n13 = "Constraint - def_spike05_l (PPPPPPP)"!(
    ComponentPose=@n12
    Alpha=1
    ConstraintWeights_0=0.6305
    ConstraintWeights_1:float=0.0015
    ConstraintWeights_2:float=0.0106
    ConstraintWeights_3:float=0.0154
    ConstraintWeights_4:float=0.0154
    ConstraintWeights_5:float=0.2815
    ConstraintWeights_6:float=0.0452
   ) -> (Pose) {
    boneToModify {boneName=def_spike05_l}
    constraintSetup [
     {targetBone {boneName=hand_l}}
     {targetBone {boneName=ring_metacarpal_l}}
     {targetBone {boneName=middle_metacarpal_l}}
     {targetBone {boneName=thumb_01_l}}
     {targetBone {boneName=index_metacarpal_l}}
     {targetBone {boneName=lowerarm_twist_01_l}}
     {targetBone {boneName=deform_wrist_up_l}}
    ]
    constraintWeights [1,1,1,1,1,1,1]
   }
   n14 = "Constraint - def_spike04_r (PPPPPPP)"!(
    ComponentPose=@n13
    Alpha=1
    ConstraintWeights_0=0.2001
    ConstraintWeights_1:float=0.5876
    ConstraintWeights_2:float=0.0506
    ConstraintWeights_3:float=0.092
    ConstraintWeights_4:float=0.0023
    ConstraintWeights_5:float=0.0003
    ConstraintWeights_6:float=0.0671
   ) -> (Pose) {
    boneToModify {boneName=def_spike04_r}
    constraintSetup [
     {targetBone {boneName=lowerarm_twist_01_r}}
     {targetBone {boneName=hand_r}}
     {targetBone {boneName=pinky_metacarpal_r}}
     {targetBone {boneName=ring_metacarpal_r}}
     {targetBone {boneName=middle_metacarpal_r}}
     {targetBone {boneName=index_metacarpal_r}}
     {targetBone {boneName=deform_wrist_up_r}}
    ]
    constraintWeights [1,1,1,1,1,1,1]
   }
   n15 = "Constraint - def_spike05_r (PPPPPPP)"!(
    ComponentPose=@n14
    Alpha=1
    ConstraintWeights_0=0.2815
    ConstraintWeights_1:float=0.6305
    ConstraintWeights_2:float=0.0015
    ConstraintWeights_3:float=0.0106
    ConstraintWeights_4:float=0.0154
    ConstraintWeights_5:float=0.0154
    ConstraintWeights_6:float=0.0452
   ) -> (Pose) {
    boneToModify {boneName=def_spike05_r}
    constraintSetup [
     {targetBone {boneName=lowerarm_twist_01_r}}
     {targetBone {boneName=hand_r}}
     {targetBone {boneName=ring_metacarpal_r}}
     {targetBone {boneName=middle_metacarpal_r}}
     {targetBone {boneName=index_metacarpal_r}}
     {targetBone {boneName=thumb_01_r}}
     {targetBone {boneName=deform_wrist_up_r}}
    ]
    constraintWeights [1,1,1,1,1,1,1]
   }
   n16 = Anim Dynamics - Bone: dyn_footbelt_up_l(ComponentPose=@n15, Alpha=1, ExternalForce="0,0,0") {
    angularDampingOverride=0.9
    angularSpringConstant=300
    bAngularSpring=true
    bOverrideAngularDamping=true
    bUseGravityOverride=true
    simulationSpace=RootRelative
    boundBone {boneName=dyn_footbelt_up_l}
    componentAppliedLinearAccClamp=<3000 3000 3000>
    componentLinearAccScale=<0.3 0.3 0.3>
    componentLinearVelScale=<0.3 0.3 0.3>
    gravityOverride {z=-980}
    physicsBodyDefinitions [{
      boundBone {boneName=dyn_footbelt_up_l}
      boxExtents=<11 11 11>
      constraintSetup {angularLimitsMax=<0 25 5>, angularLimitsMin=<0 -5 -5>, angularTarget=<1 0 0>}
      localJointOffset=<0.7 0 0>
     }]
    relativeSpaceBone {boneName=calf_twist_02_l}
    simSpaceSettings {maxAngularAcceleration=1500, maxAngularVelocity=1500, simSpaceAngularAlpha=0.2}
   }
   n17 = Anim Dynamics - Bone: dyn_footbelt_down_l(ComponentPose=@n16, Alpha=1, ExternalForce="0,0,0") {
    angularDampingOverride=0.9
    angularSpringConstant=300
    bAngularSpring=true
    bOverrideAngularDamping=true
    bUseGravityOverride=true
    simulationSpace=RootRelative
    boundBone {boneName=dyn_footbelt_down_l}
    componentAppliedLinearAccClamp=<3000 3000 3000>
    componentLinearAccScale=<0.3 0.3 0.3>
    componentLinearVelScale=<0.3 0.3 0.3>
    gravityOverride {z=-980}
    physicsBodyDefinitions [{
      boundBone {boneName=dyn_footbelt_down_l}
      boxExtents=<12 12 12>
      constraintSetup {angularLimitsMax=<0 25 5>, angularLimitsMin=<0 -5 -5>, angularTarget=<1 0 0>}
      localJointOffset=<1 0 0>
     }]
    relativeSpaceBone {boneName=calf_twist_01_l}
    simSpaceSettings {maxAngularAcceleration=1500, maxAngularVelocity=1500, simSpaceAngularAlpha=0.2}
   }
   n18 = Anim Dynamics - Bone: dyn_footbelt_up_r(ComponentPose=@n17, Alpha=1, ExternalForce="0,0,0") {
    angularDampingOverride=0.9
    angularSpringConstant=300
    bAngularSpring=true
    bOverrideAngularDamping=true
    bUseGravityOverride=true
    simulationSpace=RootRelative
    boundBone {boneName=dyn_footbelt_up_r}
    componentAppliedLinearAccClamp=<3000 3000 3000>
    componentLinearAccScale=<0.3 0.3 0.3>
    componentLinearVelScale=<0.3 0.3 0.3>
    gravityOverride {z=-980}
    physicsBodyDefinitions [{
      boundBone {boneName=dyn_footbelt_up_r}
      boxExtents=<11.5 11.5 11.5>
      constraintSetup {angularLimitsMax=<0 5 5>, angularLimitsMin=<0 -25 -5>, angularTarget=<1 0 0>}
      localJointOffset=<0.9 0 0>
     }]
    relativeSpaceBone {boneName=calf_twist_02_r}
    simSpaceSettings {maxAngularAcceleration=1500, maxAngularVelocity=1500, simSpaceAngularAlpha=0.2}
   }
   n19 = Anim Dynamics - Bone: dyn_footbelt_down_r(ComponentPose=@n18, Alpha=1, ExternalForce="0,0,0") {
    angularDampingOverride=0.9
    angularSpringConstant=300
    bAngularSpring=true
    bOverrideAngularDamping=true
    bUseGravityOverride=true
    simulationSpace=RootRelative
    boundBone {boneName=dyn_footbelt_down_r}
    componentAppliedLinearAccClamp=<3000 3000 3000>
    componentLinearAccScale=<0.3 0.3 0.3>
    componentLinearVelScale=<0.3 0.3 0.3>
    gravityOverride {z=-980}
    physicsBodyDefinitions [{
      boundBone {boneName=dyn_footbelt_down_r}
      boxExtents=<12 12 12>
      constraintSetup {angularLimitsMax=<0 5 5>, angularLimitsMin=<0 -25 -5>, angularTarget=<1 0 0>}
      localJointOffset=<0.8 0 0>
     }]
    relativeSpaceBone {boneName=calf_twist_01_r}
    simSpaceSettings {maxAngularAcceleration=1500, maxAngularVelocity=1500, simSpaceAngularAlpha=0.2}
   }
   Save cached pose 'ClothingDynamics'(Pose=Component To Local(ComponentPose=@n19)) {cacheName=ClothingDynamics}
   n20 = "Transform (Modify) Bone - Bone: dyn_upperTorso_l"(
    ComponentPose=Local To Component(LocalPose=Use cached pose 'ClothingDynamics'())
    Alpha=1
    Translation="0,0,1"
    Rotation="0,0,0"
    Scale="1,1,1"
   ) {translationMode=BMM_Additive, translationSpace=BCS_BoneSpace, boneToModify {boneName=dyn_upperTorso_l}, translation {z=1}}
   n21 = "Transform (Modify) Bone - Bone: dyn_upperTorso_r"(ComponentPose=@n20, Alpha=1, Translation="0,0,1", Rotation="0,0,0", Scale="1,1,1") {translationMode=BMM_Additive, translationSpace=BCS_BoneSpace, boneToModify {boneName=dyn_upperTorso_r}, translation {z=1}}
   n22 = Get Sim Space Settings() -> (Sim Space Settings:Anim Phys Sim Space Settings Structure)
   n23 = Anim Dynamics - Bone: dyn_upperTorso_r!(ComponentPose=@n21, Alpha=1, AngularSpringConstant:float=3000, SimSpaceSettings:Anim Phys Sim Space Settings Structure=@n22) -> (Pose) {
    angularBiasOverride=1
    angularDampingOverride=0.9999
    angularSpringConstant=5000
    bAngularSpring=true
    bGravityOverrideInSimSpace=true
    bOverrideAngularBias=true
    bOverrideAngularDamping=true
    bUseGravityOverride=true
    linearDampingOverride=0.9999
    linearSpringConstant=100
    simulationSpace=RootRelative
    boundBone {boneName=dyn_upperTorso_r}
    componentAppliedLinearAccClamp=<500 500 500>
    gravityOverride {z=-980}
    physicsBodyDefinitions [{
      boundBone {boneName=dyn_upperTorso_r}
      boxExtents=<24 24 24>
      constraintSetup {angularLimitsMax=<0 1 3>, angularLimitsMin=<0 -5 -3>, angularTarget=<1 0 0>}
      localJointOffset=<2 0 0>
     }]
    relativeSpaceBone {boneName=spine_04}
    simSpaceSettings {maxAngularAcceleration=45, maxAngularVelocity=8, simSpaceAngularAlpha=0.25}
   }
   n24 = Anim Dynamics - Bone: dyn_upperTorso_l!(ComponentPose=@n23, Alpha=1, AngularSpringConstant:float=3000, SimSpaceSettings:Anim Phys Sim Space Settings Structure=@n22) -> (Pose) {
    angularBiasOverride=1
    angularDampingOverride=0.9999
    angularSpringConstant=5000
    bAngularSpring=true
    bGravityOverrideInSimSpace=true
    bOverrideAngularBias=true
    bOverrideAngularDamping=true
    bUseGravityOverride=true
    linearDampingOverride=0.9999
    linearSpringConstant=100
    simulationSpace=RootRelative
    boundBone {boneName=dyn_upperTorso_l}
    componentAppliedLinearAccClamp=<500 500 500>
    gravityOverride {z=-980}
    physicsBodyDefinitions [{
      boundBone {boneName=dyn_upperTorso_l}
      boxExtents=<22 22 22>
      constraintSetup {angularLimitsMax=<0 1 3>, angularLimitsMin=<0 -5 -3>, angularTarget=<1 0 0>}
      localJointOffset=<2 0 0>
     }]
    relativeSpaceBone {boneName=spine_04}
    simSpaceSettings {maxAngularAcceleration=45, maxAngularVelocity=8, simSpaceAngularAlpha=0.25}
   }
   n25 = Anim Dynamics - Bone: dyn_upperTorso_r!(ComponentPose=@n24, Alpha=1, GravityOverride:Vector="0,0,-200", LinearSpringConstant:float=180) -> (Pose) {
    angularBiasOverride=1
    angularDampingOverride=0.99
    angularSpringConstant=10000
    bGravityOverrideInSimSpace=true
    bLinearSpring=true
    bOverrideAngularBias=true
    bOverrideLinearDamping=true
    bUseGravityOverride=true
    linearDampingOverride=1
    linearSpringConstant=180
    simulationSpace=RootRelative
    boundBone {boneName=dyn_upperTorso_r}
    componentAppliedLinearAccClamp=<5 5 5>
    gravityOverride {z=-200}
    physicsBodyDefinitions [{
      boundBone {boneName=dyn_upperTorso_r}
      boxExtents=<80 80 80>
      constraintSetup {angularLimitsMax=<0 0 0>, angularLimitsMin=<0 0 0>, angularTarget=<1 0 0>, linearAxesMax=<0 0.5 1>, linearAxesMin=<0 -0.5 -0.5>}
      localJointOffset=<0.2 0 0>
     }]
    relativeSpaceBone {boneName=spine_04}
    simSpaceSettings {maxAngularAcceleration=50, maxAngularVelocity=10, simSpaceAngularAlpha=0.1}
   }
   n26 = Anim Dynamics - Bone: dyn_upperTorso_l!(ComponentPose=@n25, Alpha=1, GravityOverride:Vector="0,0,-200", LinearSpringConstant:float=180) -> (Pose) {
    angularBiasOverride=1
    angularDampingOverride=0.99
    angularSpringConstant=10000
    bGravityOverrideInSimSpace=true
    bLinearSpring=true
    bOverrideAngularBias=true
    bOverrideLinearDamping=true
    bUseGravityOverride=true
    linearDampingOverride=1
    linearSpringConstant=180
    simulationSpace=RootRelative
    boundBone {boneName=dyn_upperTorso_l}
    componentAppliedLinearAccClamp=<5 5 5>
    gravityOverride {z=-200}
    physicsBodyDefinitions [{
      boundBone {boneName=dyn_upperTorso_l}
      boxExtents=<90 90 90>
      constraintSetup {angularLimitsMax=<0 0 0>, angularLimitsMin=<0 0 0>, angularTarget=<1 0 0>, linearAxesMax=<0 0.5 1>, linearAxesMin=<0 -0.5 -0.5>}
      localJointOffset=<0.2 0 0>
     }]
    relativeSpaceBone {boneName=spine_04}
    simSpaceSettings {maxAngularAcceleration=50, maxAngularVelocity=10, simSpaceAngularAlpha=0.1}
   }
   Save cached pose 'BodyDynamics'(Pose=Component To Local(ComponentPose=@n26)) {cacheName=BodyDynamics}
  ]}
 EventGraph {nodes [Event BlueprintUpdateAnimation()]}
]
variables [Sim Space Settings {
  type=AnimPhysSimSpaceSettings
  default {maxAngularAcceleration=45, maxAngularVelocity=8, simSpaceAngularAlpha=0.1, externalAngularVelocity=<0 0 0>}
 }]
