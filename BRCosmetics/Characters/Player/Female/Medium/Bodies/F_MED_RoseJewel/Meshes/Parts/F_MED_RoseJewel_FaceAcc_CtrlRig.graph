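# Control Rig for the F_MED_RoseJewel face accessory part that applies procedural corrective motion to its hair bones as the body moves.
# Parent constraints blend the main and rear-side hair master bones mostly to the head, partly to the neck.
# Head rotation relative to spine_05 is remapped into a weight for back-hair corrective transforms.
# Spherical pose readers on each clavicle weight side-hair corrective transforms, with debug drawing left enabled.
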ControlRigBlueprint /BRCosmetics/Characters/Player/Female/Medium/Bodies/F_MED_RoseJewel/Meshes/Parts/F_MED_RoseJewel_FaceAcc_CtrlRig
graphs [
 RigVMFunctionLibrary {links [], nodes []}
 RigVMModel {nodes [
   RigUnit_BeginExecution = &RigUnit_BeginExecution()
   &RigUnit_ParentConstraint(
    AdvancedSettings=(InterpolationType=Average,RotationOrderForFilter=XZY)
    Child=(Type=Bone,Name="MSTR_hair")
    Parents=((Item=(Type=Bone,Name="head"),Weight=0.85),(Item=(Type=Bone,Name="neck_01"),Weight=0.07),(Item=(Type=Bone,Name="neck_02"),Weight=0.07))
    Weight=1
    ExecutePin=@RigUnit_BeginExecution.ExecutePin
   )
   > &RigUnit_ParentConstraint(
    AdvancedSettings=(InterpolationType=Average,RotationOrderForFilter=XZY)
    Child=(Type=Bone,Name="MSTR_hair01_rear_side_r")
    Parents=((Item=(Type=Bone,Name="head"),Weight=0.9),(Item=(Type=Bone,Name="neck_01"),Weight=0),(Item=(Type=Bone,Name="neck_02"),Weight=0.1))
    Weight=1
   )
   > &RigUnit_ParentConstraint(
    AdvancedSettings=(InterpolationType=Average,RotationOrderForFilter=XZY)
    Child=(Type=Bone,Name="MSTR_hair01_rear_side_l")
    Parents=((Item=(Type=Bone,Name="head"),Weight=0.9),(Item=(Type=Bone,Name="neck_01"),Weight=0),(Item=(Type=Bone,Name="neck_02"),Weight=0.1))
    Weight=1
   )
   RigVMFunction_MathQuaternionToEuler = &RigVMFunction_MathQuaternionToEuler(
    RotationOrder=ZYX
    Value=&RigUnit_GetRelativeTransformForItem(Child=(Type=Bone,Name="head"), Parent=(Type=Bone,Name="spine_05"), bChildInitial=False, bParentInitial=False).RelativeTransform.Rotation
   )
   Remap = &RigVMFunction_MathDoubleRemap(SourceMaximum=-45, SourceMinimum=0, TargetMaximum=1, TargetMinimum=0, Value=@RigVMFunction_MathQuaternionToEuler.Result.Z)
   > &RigUnit_ModifyTransforms(
    ItemToModify="((Item=(Type=Bone,Name=\"MSTR_hair\"),Transform=(Rotation=(X=0,Y=0,Z=-0.3827,W=0.9239),Translation=(X=-2,Y=2,Z=0),Scale3D=(X=0.8,Y=0.8,Z=0.8))),(Item=(Type=Bone,Name=\"def_hair_bk_corr\"),Transform=(Rotation=(X=0,Y=0,Z=-0.6428,W=0.766),Translation=(X=6,Y=-5,Z= ...(truncated)"
    WeightMinimum=1
    Weight=@Remap.Result
   )
   > n1 = &RigUnit_SphericalPoseReader(
    ActiveRegionScaleFactors=(PositiveWidth=0.5,NegativeWidth=1,PositiveHeight=1,NegativeHeight=1)
    ActiveRegionSize=0.1
    Debug=(bDrawDebug=true,bDraw2D=false,bDrawLocalAxes=False,DebugScale=25,DebugSegments=20,DebugThickness=0.25)
    DriverAxis=(X=-1,Y=0,Z=0)
    DriverItem=(Type=Bone,Name="clavicle_r")
    FalloffRegionScaleFactors=(PositiveWidth=1,NegativeWidth=1,PositiveHeight=1,NegativeHeight=1)
    FalloffSize=0.2
    FlipHeightScaling=False
    FlipWidthScaling=False
    OptionalParentItem=(Type=Bone,Name="None")
    RotationOffset=(X=-15,Y=135,Z=0)
   )
   > &RigUnit_ModifyTransforms(
    ItemToModify="((Item=(Type=Bone,Name=\"MSTR_hair02_rear_side_r\"),Transform=(Rotation=(X=0.171,Y=0.0302,Z=0.171,W=0.9698),Translation=(X=0,Y=1.5,Z=0),Scale3D=(X=0.9,Y=0.9,Z=0.9))),(Item=(Type=Bone,Name=\"MSTR_hair_side_r\"),Transform=(Rotation=(X=-0.0864,Y=0.13,Z=0.0114,W=0 ...(truncated)"
    WeightMinimum=0
    Weight=@n1.OutputParam
   )
   > n2 = &RigUnit_SphericalPoseReader(
    ActiveRegionScaleFactors=(PositiveWidth=0.5,NegativeWidth=1,PositiveHeight=1,NegativeHeight=1)
    ActiveRegionSize=0.1
    Debug=(bDrawDebug=true,bDraw2D=false,bDrawLocalAxes=False,DebugScale=25,DebugSegments=20,DebugThickness=0.25)
    DriverAxis=(X=1,Y=0,Z=0)
    DriverItem=(Type=Bone,Name="clavicle_l")
    FalloffRegionScaleFactors=(PositiveWidth=1,NegativeWidth=1,PositiveHeight=1,NegativeHeight=1)
    FalloffSize=0.2
    FlipHeightScaling=False
    FlipWidthScaling=False
    OptionalParentItem=(Type=Bone,Name="None")
    RotationOffset=(X=15,Y=-45,Z=0)
   )
   > &RigUnit_ModifyTransforms(
    ItemToModify="((Item=(Type=Bone,Name=\"MSTR_hair02_rear_side_l\"),Transform=(Rotation=(X=-0.0872,Y=0,Z=0,W=0.9962),Translation=(X=0,Y=-1.5,Z=0),Scale3D=(X=0.9,Y=0.9,Z=0.9))),(Item=(Type=Bone,Name=\"MSTR_hair_side_l\"),Transform=(Rotation=(X=0.1722,Y=-0.1285,Z=0.0227,W=0.976 ...(truncated)"
    WeightMinimum=0
    Weight=@n2.OutputParam
   )
  ]}
]
hierarchy [
 root {loc=<0 0 0>, rot=<0 0 0 1>}
 pelvis {parent=root, loc=<0 -0.5301 98.0571>, rot=<0.0127 0.707 -0.0127 -0.707>}
 spine_01 {parent=pelvis, loc=<6.1208 -0.9328 0>, rot=<0 0 0.034 -0.9994>}
 spine_02 {parent=spine_01, loc=<6.5997 -0.2095 0>, rot=<0 0 0.0238 -0.9997>}
 spine_03 {parent=spine_02, loc=<6.4914 -0.1237 0>, rot=<0 0 0.0149 -0.9999>}
 spine_04 {parent=spine_03, loc=<6.9598 0.0794 0>, rot=<0 0 -0.0132 -0.9999>}
 spine_05 {parent=spine_04, loc=<6.9987 0 0>, rot=<0 0 -0.0015 -1>}
 neck_01 {parent=spine_05, loc=<11.3138 0.1529 0>, rot=<0 0 -0.2003 -0.9797>}
 neck_02 {parent=neck_01, loc=<4.5627 0 0>, rot=<0 0 0.0504 -0.9987>}
 head {parent=neck_02, loc=<5.3914 0 0>, rot=<0 0 0.115 -0.9934>}
 MSTR_hair {parent=head, loc=<4.3329 -7.7716 -0.331>, rot=<0.2213 -0.9752 0 0>}
 dyn_hair01_r {parent=MSTR_hair, loc=<-0.2857 0.7829 -4.3801>, rot=<-0.2235 -0.0581 0.036 -0.9723>}
 dyn_hair02_r {parent=dyn_hair01_r, loc=<5.2073 0.6218 -0.1148>, rot=<0.035 0.0046 -0.3203 -0.9466>}
 dyn_hair01_l {parent=MSTR_hair, loc=<-0.2859 0.7828 3.7181>, rot=<0.1955 0.0802 0.0287 -0.977>}
 dyn_hair02_l {parent=dyn_hair01_l, loc=<5.374 0.378 -0.0444>, rot=<-0.035 -0.0046 -0.3203 -0.9466>}
 dyn_hair01_m {parent=MSTR_hair, loc=<0 0 0>, rot=<0 0 0 1>}
 dyn_hair02_m {parent=dyn_hair01_m, loc=<5.5094 0.2614 0>, rot=<0 0 -0.2767 -0.961>}
 MSTR_curl_fr_r {parent=head, loc=<9.1716 8.9801 2.19>, rot=<0.0987 -0.3445 0.8996 0.2496>}
 dyn_curl_fr_01_r {parent=MSTR_curl_fr_r, loc=<0 0 0>, rot=<0 0 0 1>}
 dyn_curl_fr_02_r {parent=dyn_curl_fr_01_r, loc=<3.1385 0 0>, rot=<0.01 -0.1055 -0.2781 -0.9547>}
 dyn_curl_fr_03_r {parent=dyn_curl_fr_02_r, loc=<1.964 0 0>, rot=<0.0473 0.0515 -0.2509 -0.9655>}
 dyn_curl_fr_04_r {parent=dyn_curl_fr_03_r, loc=<1.5532 0 0>, rot=<-0.0155 -0.0934 -0.1115 -0.9892>}
 dyn_curl_fr_05_r {parent=dyn_curl_fr_01_r, loc=<3.4183 -0.218 1.7368>, rot=<-0.6945 -0.143 -0.2823 -0.6461>}
 dyn_curl_fr_06_r {parent=dyn_curl_fr_05_r, loc=<2.5784 0.3947 0.2216>, rot=<0.7536 -0.1245 -0.0634 -0.6423>}
 dyn_curl_fr_07_r {parent=dyn_curl_fr_06_r, loc=<1.5762 0 0>, rot=<-0.0184 -0.0744 -0.1477 -0.9861>}
 dyn_curl_fr_08_r {parent=dyn_curl_fr_05_r, loc=<2.0491 -0.5617 0.225>, rot=<0.6774 -0.1452 -0.0796 -0.7168>}
 MSTR_curl_fr_l {parent=head, loc=<8.8383 9.4437 -2.3172>, rot=<-0.018 0.0733 0.9704 0.2294>}
 dyn_curl_fr_01_l {parent=MSTR_curl_fr_l, loc=<0 0 0>, rot=<0 0 0 1>}
 dyn_curl_fr_02_l {parent=dyn_curl_fr_01_l, loc=<2.8673 0 0>, rot=<0.0362 0.1927 -0.2845 -0.9384>}
 dyn_curl_fr_03_l {parent=dyn_curl_fr_02_l, loc=<2.7424 0 0>, rot=<-0.0423 0.0893 -0.0679 -0.9928>}
 dyn_curl_fr_04_l {parent=dyn_curl_fr_03_l, loc=<1.0066 -0.0122 0.5477>, rot=<0.0509 0.1304 -0.0899 -0.9861>}
 dyn_curl_fr_05_l {parent=dyn_curl_fr_03_l, loc=<0.857 -0.0382 -0.6657>, rot=<-0.0843 0.1394 -0.0395 -0.9859>}
 dyn_curl_fr_06_l {parent=dyn_curl_fr_02_l, loc=<1.7409 0.1253 1.5581>, rot=<0.0328 0.0666 -0.0896 -0.9932>}
 MSTR_curl_side_l {parent=head, loc=<5.9637 8.6276 -6.0005>, rot=<0.1121 0.1746 0.978 -0.0209>}
 dyn_curl_side_01_l {parent=MSTR_curl_side_l, loc=<0 0 0>, rot=<0 0 0 1>}
 dyn_curl_side_02_l {parent=dyn_curl_side_01_l, loc=<3.0639 0.1778 -0.2787>, rot=<-0.0168 0.0455 -0.0557 -0.9973>}
 dyn_curl_side_03_l {parent=dyn_curl_side_01_l, loc=<1.3932 -0.6633 1.4404>, rot=<-0.0425 0.1234 -0.0022 -0.9914>}
 MSTR_hair_side_r {parent=head, loc=<6.5093 3.3332 6.6207>, rot=<0.1209 -0.6289 0.7676 -0.0257>}
 dyn_hair_side01_r {parent=MSTR_hair_side_r, loc=<0 0 0>, rot=<0 0 0 1>}
 dyn_hair_side02_r {parent=dyn_hair_side01_r, loc=<4.1099 0 0>, rot=<0.0029 0.0125 0.1442 -0.9895>}
 dyn_hair_side03_r {parent=dyn_hair_side02_r, loc=<3.529 0 0>, rot=<0.0005 -0.0473 0.0569 -0.9973>}
 MSTR_hair_side_l {parent=head, loc=<6.4239 3.3657 -6.3377>, rot=<-0.135 0.5715 0.8094 -0.0087>}
 dyn_hair_side01_l {parent=MSTR_hair_side_l, loc=<0 0 0>, rot=<0 0 0 1>}
 dyn_hair_side02_l {parent=dyn_hair_side01_l, loc=<4.3358 0 0>, rot=<-0.0738 -0.049 0.0292 -0.9956>}
 dyn_hair_side03_l {parent=dyn_hair_side02_l, loc=<3.0817 0.2563 1.4575>, rot=<0.1521 0.2168 0.1763 -0.948>}
 dyn_hair_side04_l {parent=dyn_hair_side02_l, loc=<2.4269 -0.8674 -1.5085>, rot=<-0.0292 0.1461 0.1976 -0.9689>}
 FACIAL_C_FacialRoot {parent=head, loc=<2.1524 -0.746 -0.0002>, rot=<0.4879 -0.5118 -0.4879 -0.5118>}
 FACIAL_R_Ear {parent=FACIAL_C_FacialRoot, loc=<-6.397 0.7007 0.5987>, rot=<-0.0266 0.0007 0.0262 -0.9993>}
 FACIAL_R_Masseter {parent=FACIAL_C_FacialRoot, loc=<-5.9638 2.7462 4.0402>, rot=<0.069 -0.1129 -0.0079 -0.9912>}
 FACIAL_L_LipCorner {parent=FACIAL_C_FacialRoot, loc=<2.0866 6.314 7.3005>, rot=<-0.0003 -0.2079 -0.0001 -0.9782>}
 FACIAL_C_MouthUpper {parent=FACIAL_C_FacialRoot, loc=<-0.0004 5.7962 6.767>, rot=<-0.0003 0 0 -1>}
 FACIAL_C_LipUpper {parent=FACIAL_C_MouthUpper, loc=<-0.0006 -0.0459 2.0748>, rot=<-0.0872 0 0 -0.9962>}
 FACIAL_L_LipUpper {parent=FACIAL_C_MouthUpper, loc=<0.9017 0.0116 1.8659>, rot=<-0.0778 -0.1601 -0.0379 -0.9833>}
 FACIAL_R_LipUpper {parent=FACIAL_C_MouthUpper, loc=<-0.9034 0.0115 1.8642>, rot=<-0.0778 0.1601 0.0379 -0.9833>}
 FACIAL_L_LipUpperOuter {parent=FACIAL_C_MouthUpper, loc=<1.6673 0.2014 1.2461>, rot=<-0.0984 -0.2293 -0.0983 -0.9634>}
 FACIAL_R_LipUpperOuter {parent=FACIAL_C_MouthUpper, loc=<-1.6748 0.1997 1.2439>, rot=<-0.0984 0.2293 0.0983 -0.9634>}
 FACIAL_R_LipCorner {parent=FACIAL_C_MouthUpper, loc=<-2.0889 0.5137 0.5357>, rot=<0 0.2078 0 -0.9782>}
 FACIAL_C_LowerLipRotation {parent=FACIAL_C_FacialRoot, loc=<0 3.2469 1.8984>, rot=<0.3418 0 0 -0.9398>}
 FACIAL_C_MouthLower {parent=FACIAL_C_LowerLipRotation, loc=<-0.0005 -1.1737 5.3689>, rot=<-0.342 0 0 -0.9397>}
 FACIAL_C_LipLower {parent=FACIAL_C_MouthLower, loc=<0.0002 1.3394 1.4412>, rot=<0.0872 0 0 -0.9962>}
 FACIAL_L_LipLower {parent=FACIAL_C_MouthLower, loc=<0.7887 1.291 1.2451>, rot=<0.0927 -0.1294 0.0272 -0.9869>}
 FACIAL_R_LipLower {parent=FACIAL_C_MouthLower, loc=<-0.7886 1.2901 1.2452>, rot=<0.0927 0.1294 -0.0272 -0.9869>}
 FACIAL_L_LipLowerOuter {parent=FACIAL_C_MouthLower, loc=<1.5159 0.9197 0.8843>, rot=<0.0913 -0.2321 0.0318 -0.9679>}
 FACIAL_R_LipLowerOuter {parent=FACIAL_C_MouthLower, loc=<-1.5169 0.9174 0.8821>, rot=<0.0913 0.2321 -0.0318 -0.9679>}
 FACIAL_C_TeethLower {parent=FACIAL_C_LowerLipRotation, loc=<-0.0006 -1.1826 6.3099>, rot=<-0.342 0 0 -0.9397>}
 FACIAL_C_Tongue1 {parent=FACIAL_C_TeethLower, loc=<0.002 -0.04 -4.5127>, rot=<-0.2588 0 0 -0.9659>}
 FACIAL_C_Tongue2 {parent=FACIAL_C_Tongue1, loc=<-0.0007 0.4082 1.3222>, rot=<0.2588 0 0 -0.9659>}
 FACIAL_C_Tongue3 {parent=FACIAL_C_Tongue2, loc=<-0.0004 0.2368 1.1967>, rot=<0.1305 0 0 -0.9914>}
 FACIAL_C_Tongue4 {parent=FACIAL_C_Tongue3, loc=<0.0001 -0.1085 1.0287>, rot=<0.0436 0 0 -0.999>}
 FACIAL_C_TongueUpper3 {parent=FACIAL_C_Tongue3, loc=<0.0002 -0.3448 -0.1567>, rot=<0 0 0 1>}
 FACIAL_C_TongueLower3 {parent=FACIAL_C_Tongue3, loc=<0 0.7848 0.0529>, rot=<0 0 0 1>}
 FACIAL_L_TongueSide3 {parent=FACIAL_C_Tongue3, loc=<1.4493 0.0421 0.1383>, rot=<0 0 0 1>}
 FACIAL_R_TongueSide3 {parent=FACIAL_C_Tongue3, loc=<-1.4491 0.0416 0.1365>, rot=<0 0 0 1>}
 FACIAL_C_TongueUpper2 {parent=FACIAL_C_Tongue2, loc=<0.0001 -0.5632 -0.1252>, rot=<0 0 0 1>}
 FACIAL_L_TongueSide2 {parent=FACIAL_C_Tongue2, loc=<1.714 0.0841 0.2399>, rot=<0 0 0 1>}
 FACIAL_R_TongueSide2 {parent=FACIAL_C_Tongue2, loc=<-1.7145 0.0837 0.2376>, rot=<0 0 0 1>}
 FACIAL_R_ChinSide {parent=FACIAL_C_FacialRoot, loc=<-2.3102 7.9437 6.9>, rot=<0.0999 0.3342 -0.1746 -0.9208>}
 FACIAL_C_Chin {parent=FACIAL_C_FacialRoot, loc=<0.0004 8.5949 6.7589>, rot=<0.0866 0 0 -0.9962>}
 FACIAL_L_ChinSide {parent=FACIAL_C_FacialRoot, loc=<2.3129 7.9451 6.9003>, rot=<0.0999 -0.3343 0.1746 -0.9208>}
 def_ear_piercing {parent=head, loc=<-0.6146 -0.0664 6.2963>, rot=<0.5982 -0.5453 0.3991 0.4306>}
 def_lip_piercing {parent=head, loc=<-4.3707 7.3458 0.9275>, rot=<0.7223 0.2987 0.488 0.3885>}
 MSTR_hair01_rear_side_r {parent=head, loc=<0.8355 -4.0973 8.3004>, rot=<0.4388 -0.8923 0.1009 0.0336>}
 dyn_hair01_rear_side_r {parent=MSTR_hair01_rear_side_r, loc=<0 0 0>, rot=<0 0 0 1>}
 MSTR_hair02_rear_side_r {parent=head, loc=<-1.1513 -2.3802 6.7127>, rot=<0.0515 -0.8263 0.55 -0.1094>}
 dyn_hair02_rear_side_r {parent=MSTR_hair02_rear_side_r, loc=<0 0 0>, rot=<0 0 0 1>}
 MSTR_hair01_rear_side_l {parent=head, loc=<0.3551 -2.664 -8.1287>, rot=<0.6405 -0.7605 -0.0686 -0.0824>}
 dyn_hair01_rear_side_l {parent=MSTR_hair01_rear_side_l, loc=<0 0 0>, rot=<0 0 0 1>}
 MSTR_hair02_rear_side_l {parent=head, loc=<-1.1385 -2.3329 -6.6068>, rot=<0.0391 -0.6034 0.7963 0.0157>}
 dyn_hair02_rear_side_l {parent=MSTR_hair02_rear_side_l, loc=<0 0 0>, rot=<0 0 0 1>}
 def_hair_bk_corr {parent=head, loc=<3.2898 -8.3073 -0.9029>, rot=<0.7329 -0.6804 0 0>}
 MSTR_hair_inbetween {parent=head, loc=<2.6532 -0.4774 8.3867>, rot=<0.3502 -0.6266 0.6941 0.0537>}
 dyn_hair_inbetween_01 {parent=MSTR_hair_inbetween, loc=<0 0 0>, rot=<0 0 0 1>}
 dyn_hair_inbetween_02 {parent=dyn_hair_inbetween_01, loc=<2.9197 -0.3889 0.4871>, rot=<0.154 0.1555 0.064 -0.9737>}
 dyn_fur {parent=spine_05, loc=<8.4626 -9.1205 0.3343>, rot=<0.8019 -0.5972 -0.0068 -0.0154>}
 dyn_fur_rear_l {parent=spine_05, loc=<9.1739 -8.0284 -7.0873>, rot=<0.7774 -0.6044 -0.0947 -0.1459>}
 clavicle_l {parent=spine_05, loc=<5.4831 -0.1053 -3.2946>, rot=<0.0465 -0.7289 -0.009 -0.6829>}
 upperarm_l {parent=clavicle_l, loc=<9.8658 -1.2136 -0.4105>, rot=<0.0164 -0.3245 -0.0269 -0.9454>}
 clavicle_r {parent=spine_05, loc=<5.4831 -0.1053 3.2946>, rot=<0.7289 0.0465 -0.6829 0.009>}
 upperarm_r {parent=clavicle_r, loc=<-9.8658 1.2136 0.4105>, rot=<0.0164 -0.3245 -0.0269 -0.9454>}
]
