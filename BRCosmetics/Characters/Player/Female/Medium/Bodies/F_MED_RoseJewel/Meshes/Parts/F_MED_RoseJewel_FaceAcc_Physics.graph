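# Physics asset for the RoseJewel female medium character's face accessory mesh, setting up physics for dynamic hair bones.
# Capsule hair bodies form short chains hung from small MSTR_hair anchor spheres, with tight swing limits and mostly locked twist.
# Extra angularly free constraints cross-link the middle strand to its left and right neighbours, and the right side to the in-between strand.
# Unconstrained head, neck and upper-spine bodies supply body collision shapes.
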PhysicsAsset /BRCosmetics/Characters/Player/Female/Medium/Bodies/F_MED_RoseJewel/Meshes/Parts/F_MED_RoseJewel_FaceAcc_Physics
bodies {
 MSTR_hair [{shapeType=Sphere, radius=0.505, shapeName=MSTR_hair_sphere}]
 MSTR_hair01_rear_side_l [{shapeType=Sphere, radius=0.505, shapeName=MSTR_hair01_rear_side_l_sphere}]
 MSTR_hair01_rear_side_r [{shapeType=Sphere, radius=0.505, shapeName=MSTR_hair01_rear_side_r_sphere}]
 MSTR_hair02_rear_side_l [{shapeType=Sphere, radius=0.505, shapeName=MSTR_hair02_rear_side_l_sphere}]
 MSTR_hair02_rear_side_r [{shapeType=Sphere, radius=0.505, shapeName=MSTR_hair02_rear_side_r_sphere}]
 MSTR_hair_inbetween [{shapeType=Sphere, radius=0.505, shapeName=MSTR_hair_inbetween_sphere}]
 MSTR_hair_side_l [{shapeType=Sphere, radius=0.505, shapeName=MSTR_hair_side_l_sphere}]
 MSTR_hair_side_r [{shapeType=Sphere, radius=0.505, shapeName=MSTR_hair_side_r_sphere}]
 dyn_hair01_l [{
   shapeType=Capsule
   length=2.238
   radius=2.1698
   shapeName=dyn_hair01_m_capsule
   center=<0.9116 0.1215 0.6846>
   rotation=<rot -18.3805 109.6284 -0.0425>
  }]
 dyn_hair01_m [{
   shapeType=Capsule
   length=3.1735
   radius=2.1698
   shapeName=dyn_hair01_m_capsule
   center=<0.9424 -0.1014 0.1707>
   rotation=<rot 2.3618 -90 -1.4925>
  }]
 dyn_hair01_r [{
   shapeType=Capsule
   length=2.5798
   radius=2.1698
   shapeName=dyn_hair01_m_capsule
   center=<0.9984 0.1641 0.0238>
   rotation=<rot -7.9121 -91.3268 7.2868>
  }]
 dyn_hair01_rear_side_l [{
   shapeType=Capsule
   length=3.2396
   radius=2.1503
   shapeName=dyn_hair01_rear_side_l_capsule
   center=<2.3863 -0.0879 0.7787>
   rotation=<rot 0 -69.1616 -76.04>
  }]
 dyn_hair01_rear_side_r [{
   shapeType=Capsule
   length=3.2396
   radius=2.1503
   shapeName=dyn_hair01_rear_side_r_capsule
   center=<2.074 -0.4581 -0.5173>
   rotation=<rot -2.1648 74.973 110.3624>
  }]
 dyn_hair02_l [{
   shapeType=Capsule
   length=2.238
   radius=2.1698
   shapeName=dyn_hair01_m_capsule
   center=<3.4513 0.9743 0.7618>
   rotation=<rot -16.8063 88.6064 7.6514>
  }]
 dyn_hair02_m [{
   shapeType=Capsule
   length=4.2109
   radius=2.1698
   shapeName=dyn_hair01_m_capsule
   center=<3.5664 0.3784 0.257>
   rotation=<rot 2.3618 -90 -1.4925>
  }]
 dyn_hair02_r [{
   shapeType=Capsule
   length=2.5798
   radius=2.1698
   shapeName=dyn_hair01_m_capsule
   center=<3.5812 1.0993 -0.1591>
   rotation=<rot -8.7431 -91.436 7.4093>
  }]
 dyn_hair02_rear_side_l [{
   shapeType=Capsule
   length=2.3275
   radius=1.8718
   shapeName=dyn_hair02_rear_side_l_capsule
   center=<2.5198 -0.0668 -0.002>
   rotation=<rot -2.9045 -90 53.5144>
  }]
 dyn_hair02_rear_side_r [{
   shapeType=Capsule
   length=2.3275
   radius=1.8718
   shapeName=dyn_hair02_rear_side_r_capsule
   center=<2.5198 0.0668 0.002>
   rotation=<rot -6.3498 -94.7029 -126.1058>
  }]
 dyn_hair_inbetween_01 [{
   shapeType=Capsule
   length=2.9008
   radius=1.6777
   shapeName=dyn_hair_inbetween_01_capsule
   center=<2.8313 -0.6818 -0.3771>
   rotation=<rot 8.2404 -79.4461 -177.2122>
  }]
 dyn_hair_inbetween_02 [{
   shapeType=Capsule
   length=2.4008
   radius=1.1777
   shapeName=dyn_hair_inbetween_01_capsule
   center=<1.2996 -0.5254 0.182>
   rotation=<rot -0.4297 -106.6708 -108.3214>
  }]
 dyn_hair_side01_l [{
   shapeType=Capsule
   length=4.8827
   radius=1.7992
   shapeName=dyn_hair_side01_l_capsule
   center=<2.2569 -1.7019 -0.8058>
   rotation=<rot 0.7849 8.2571 -13.4033>
  }]
 dyn_hair_side01_r [{
   shapeType=Capsule
   length=5.6949
   radius=1.8533
   shapeName=dyn_hair_side01_r_capsule
   center=<1.914 -1.2523 0.9393>
   rotation=<rot -0.0002 0 4.8095>
  }]
 dyn_hair_side02_l [{
   shapeType=Capsule
   length=3.1041
   radius=2.216
   shapeName=dyn_hair_side02_l_capsule
   center=<2.7213 -0.9991 -0.4213>
   rotation=<rot 1.0488 15.0589 -1.5002>
  }]
 dyn_hair_side02_r [{
   shapeType=Capsule
   length=3.7358
   radius=2.1635
   shapeName=dyn_hair_side02_r_capsule
   center=<1.6945 -0.7739 0.5957>
   rotation=<rot -0.0001 0 2.0036>
  }]
 dyn_hair_side03_r [{
   shapeType=Capsule
   length=3.1121
   radius=2.3088
   shapeName=dyn_hair_side03_r_capsule
   center=<2.1724 0.7632 1.2274>
   rotation=<rot -5.0945 29.8968 8.7795>
  }]
 head [
  {shapeType=Sphere, radius=6.4659, shapeName=head_sphere, center=<5.5766 1.6671 -0.2207>}
  {shapeType=Capsule, length=3.5, radius=6, shapeName=head_capsule, center=<-0.6185 2.8217 0>, rotation=<rot -90 -5.253 0>}
 ]
 neck_01 [{shapeType=Capsule, length=0.9199, radius=5.2843, shapeName=neck_02_capsule, center=<0.3436 1.1556 0.003>, rotation=<rot -89.5005 -0.0002 0>}]
 neck_02 [{shapeType=Capsule, length=1, radius=4.505, shapeName=neck_02_capsule, center=<1.1254 0.8695 0.0098>, rotation=<rot -89.5005 -0.0002 0>}]
 spine_05 [
  {shapeType=Capsule, length=13.505, radius=9.505, shapeName=spine_05_capsule, center=<1.6883 1.2945 0>}
  {shapeType=Capsule, length=7, radius=7, shapeName=spine_05_capsule, center=<5 -14.6914 0>}
 ]
}
constraints [
 {
  bone1Name=dyn_hair_side01_l
  bone2Name=MSTR_hair_side_l
  swing1LimitDegrees=10
  swing2LimitDegrees=5
  twistLimitDegrees=45
  twistMotion=Locked
 }
 {bone1Name=dyn_hair_side02_l, bone2Name=dyn_hair_side01_l, swing1LimitDegrees=10, swing2LimitDegrees=5, twistLimitDegrees=5}
 {
  bone1Name=dyn_hair_side01_r
  bone2Name=MSTR_hair_side_r
  swing1LimitDegrees=10
  swing2LimitDegrees=7
  twistLimitDegrees=45
  twistMotion=Locked
 }
 {
  bone1Name=dyn_hair_side02_r
  bone2Name=dyn_hair_side01_r
  swing1LimitDegrees=15
  swing2LimitDegrees=5
  twistLimitDegrees=5
  twistMotion=Locked
 }
 {
  bone1Name=dyn_hair_side03_r
  bone2Name=dyn_hair_side02_r
  swing1LimitDegrees=10
  swing2LimitDegrees=5
  twistLimitDegrees=5
  twistMotion=Locked
 }
 {
  bone1Name=dyn_hair01_rear_side_l
  bone2Name=MSTR_hair01_rear_side_l
  swing1LimitDegrees=10
  swing2LimitDegrees=10
  twistLimitDegrees=45
  twistMotion=Locked
 }
 {
  bone1Name=dyn_hair01_rear_side_r
  bone2Name=MSTR_hair01_rear_side_r
  swing1LimitDegrees=10
  swing2LimitDegrees=10
  twistLimitDegrees=45
  twistMotion=Locked
 }
 {
  bone1Name=dyn_hair02_rear_side_l
  bone2Name=MSTR_hair02_rear_side_l
  swing1LimitDegrees=10
  swing2LimitDegrees=10
  twistLimitDegrees=45
  twistMotion=Locked
 }
 {
  bone1Name=dyn_hair02_rear_side_r
  bone2Name=MSTR_hair02_rear_side_r
  swing1LimitDegrees=10
  swing2LimitDegrees=10
  twistLimitDegrees=45
  twistMotion=Locked
 }
 {bone1Name=dyn_hair01_m, bone2Name=MSTR_hair, swing1LimitDegrees=10, swing2LimitDegrees=10, twistLimitDegrees=45, twistMotion=Locked}
 {bone1Name=dyn_hair02_m, bone2Name=dyn_hair01_m, swing1LimitDegrees=15, swing2LimitDegrees=10, twistLimitDegrees=10}
 {bone1Name=dyn_hair02_r, bone2Name=dyn_hair01_r, swing1LimitDegrees=15, swing2LimitDegrees=15, twistLimitDegrees=10}
 {bone1Name=dyn_hair01_r, bone2Name=MSTR_hair, swing1LimitDegrees=10, swing2LimitDegrees=10, twistLimitDegrees=45, twistMotion=Locked}
 {bone1Name=dyn_hair01_l, bone2Name=MSTR_hair, swing1LimitDegrees=10, swing2LimitDegrees=10, twistLimitDegrees=45, twistMotion=Locked}
 {bone1Name=dyn_hair02_l, bone2Name=dyn_hair01_l, swing1LimitDegrees=15, swing2LimitDegrees=15, twistLimitDegrees=10}
 {
  bone1Name=dyn_hair02_m
  bone2Name=dyn_hair02_r
  swing1LimitDegrees=45
  swing1Motion=Free
  swing2LimitDegrees=45
  swing2Motion=Free
  twistLimitDegrees=45
  twistMotion=Free
 }
 {
  bone1Name=dyn_hair02_m
  bone2Name=dyn_hair02_l
  swing1LimitDegrees=45
  swing1Motion=Free
  swing2LimitDegrees=45
  swing2Motion=Free
  twistLimitDegrees=45
  twistMotion=Free
 }
 {
  bone1Name=dyn_hair_inbetween_01
  bone2Name=MSTR_hair_inbetween
  swing1LimitDegrees=7
  swing2LimitDegrees=5
  twistLimitDegrees=45
  twistMotion=Locked
 }
 {
  bone1Name=dyn_hair_inbetween_02
  bone2Name=dyn_hair_inbetween_01
  swing1LimitDegrees=10
  swing2LimitDegrees=10
  twistLimitDegrees=45
  twistMotion=Locked
 }
 {
  bone1Name=dyn_hair_side03_r
  bone2Name=dyn_hair_inbetween_01
  swing1LimitDegrees=45
  swing1Motion=Free
  swing2LimitDegrees=45
  swing2Motion=Free
  twistLimitDegrees=45
  twistMotion=Free
 }
]
