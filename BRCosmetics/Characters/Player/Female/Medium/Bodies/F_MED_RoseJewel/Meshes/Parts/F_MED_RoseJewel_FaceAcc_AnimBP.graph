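# Post-process-style AnimBP animating RoseJewel's face accessory part: hair curls and piercings.
# Parent-mesh pose feeds a Control Rig, weighted lip/ear piercing constraints, head-space RigidBody, then serial AnimDynamics springs for curls and ear piercing.
# dyn_curl_fr_02_r is simulated twice; hair-inbetween dynamics and head transform nodes are unconnected.
# Update event sets Gravity Override from PolkaSkate's parameter asset, but no shown node reads it.
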AnimBlueprint /BRCosmetics/Characters/Player/Female/Medium/Bodies/F_MED_RoseJewel/Meshes/Parts/F_MED_RoseJewel_FaceAcc_AnimBP
target_skeleton=&/F_MED_RoseJewel_FaceAcc_Skeleton
graphs [
 AnimGraph {nodes [
   Anim Dynamics - Bone: dyn_hair_inbetween_02(Alpha=1, ExternalForce="0,0,0") {
    angularDampingOverride=0.9
    angularSpringConstant=250
    bAngularSpring=true
    bGravityOverrideInSimSpace=true
    bOverrideAngularDamping=true
    bUseGravityOverride=true
    simulationSpace=BoneRelative
    boundBone {boneName=dyn_hair_inbetween_02}
    componentAppliedLinearAccClamp=<2000 2000 2000>
    componentLinearAccScale=<0.2 0.2 0.2>
    componentLinearVelScale=<0.2 0.2 0.2>
    gravityOverride {x=-980}
    physicsBodyDefinitions [{
      boundBone {boneName=dyn_hair_inbetween_02}
      boxExtents=<11 11 11>
      constraintSetup {angularLimitsMax=<0 10 10>, angularLimitsMin=<0 -10 -15>, angularTarget=<1 0 0>}
      localJointOffset=<0.7 0 0>
     }]
    relativeSpaceBone {boneName=head}
    simSpaceSettings {maxAngularAcceleration=1500, maxAngularVelocity=1500, simSpaceAngularAlpha=0.2}
   }
   "Transform (Modify) Bone - Bone: head"(Alpha=1, Translation="0,0,0", Rotation="0,0,0", Scale="1,1,1") {rotationMode=BMM_Additive, boneToModify {boneName=head}}
   n1 = Control Rig(Source=Copy Pose From Mesh() {bUseAttachedParent=true}) {controlRigAssetReference {blueprintRigClass=&/F_MED_RoseJewel_FaceAcc_CtrlRig.F_MED_RoseJewel_FaceAcc_CtrlRig_C}}
   n2 = "Constraint - def_lip_piercing (PPPPPPP)"!(
    ComponentPose=Local To Component(LocalPose=@n1)
    Alpha=1
    ConstraintWeights_0=0
    ConstraintWeights_1:float=0
    ConstraintWeights_2:float=0.001
    ConstraintWeights_3:float=0.1046
    ConstraintWeights_4:float=0.2731
    ConstraintWeights_5:float=0.4579
    ConstraintWeights_6:float=0.1634
   ) -> (Pose) {
    boneToModify {boneName=def_lip_piercing}
    constraintSetup [
     {targetBone {boneName=head}}
     {targetBone {boneName=None}}
     {targetBone {boneName=FACIAL_R_LipUpperOuter}}
     {targetBone {boneName=FACIAL_R_LipCorner}}
     {targetBone {boneName=FACIAL_C_LipLower}}
     {targetBone {boneName=FACIAL_R_LipLower}}
     {targetBone {boneName=FACIAL_R_LipLowerOuter}}
    ]
    constraintWeights [1,1,1,1,1,1,1]
   }
   n3 = "Constraint - def_ear_piercing (PP)"!(ComponentPose=@n2, Alpha=1, ConstraintWeights_0=0.1474, ConstraintWeights_1:float=0.8526) -> (Pose) {
    boneToModify {boneName=def_ear_piercing}
    constraintSetup [{targetBone {boneName=head}}, {targetBone {boneName=FACIAL_R_Ear}}]
    constraintWeights [1,1]
   }
   n4 = RigidBody!(ComponentPose=@n3, Alpha, OverrideWorldGravity:Vector="0,0,-980", ExternalForce) -> (Pose) {
    bClampLinearTranslationLimitToRefPose=true
    bOverrideWorldGravity=true
    overridePhysicsAsset=&/F_MED_RoseJewel_FaceAcc_Physics
    simulationSpace=BaseBoneSpace
    baseBoneRef {boneName=head}
    overrideWorldGravity {z=-980}
    simSpaceSettings {
     dampingAlpha=0
     maxAngularAcceleration=30
     maxAngularVelocity=6
     maxLinearAcceleration=1200
     maxLinearVelocity=350
     velocityScaleZ=0.7
     worldAlpha=1
    }
   }
   n5 = Anim Dynamics - Bone: dyn_curl_fr_01_l(ComponentPose=@n4, Alpha=1, ExternalForce="0,0,0") {
    angularDampingOverride=0.9
    angularSpringConstant=250
    bAngularSpring=true
    bGravityOverrideInSimSpace=true
    bOverrideAngularDamping=true
    bUseGravityOverride=true
    simulationSpace=BoneRelative
    boundBone {boneName=dyn_curl_fr_01_l}
    componentAppliedLinearAccClamp=<2000 2000 2000>
    componentLinearAccScale=<0.2 0.2 0.2>
    componentLinearVelScale=<0.2 0.2 0.2>
    gravityOverride {x=-980}
    physicsBodyDefinitions [{
      boundBone {boneName=dyn_curl_fr_01_l}
      boxExtents=<11 11 11>
      constraintSetup {angularLimitsMax=<0 5 2>, angularLimitsMin=<0 -10 -2>, angularTarget=<1 0 0>}
      localJointOffset=<0.7 0 0>
     }]
    relativeSpaceBone {boneName=head}
    simSpaceSettings {maxAngularAcceleration=1500, maxAngularVelocity=1500, simSpaceAngularAlpha=0.2}
   }
   n6 = Anim Dynamics - Bone: dyn_curl_fr_02_l(ComponentPose=@n5, Alpha=1, ExternalForce="0,0,0") {
    angularDampingOverride=0.9
    angularSpringConstant=250
    bAngularSpring=true
    bGravityOverrideInSimSpace=true
    bOverrideAngularDamping=true
    bUseGravityOverride=true
    simulationSpace=BoneRelative
    boundBone {boneName=dyn_curl_fr_02_l}
    componentAppliedLinearAccClamp=<2000 2000 2000>
    componentLinearAccScale=<0.2 0.2 0.2>
    componentLinearVelScale=<0.2 0.2 0.2>
    gravityOverride {x=-980}
    physicsBodyDefinitions [{
      boundBone {boneName=dyn_curl_fr_02_l}
      boxExtents=<10 10 10>
      constraintSetup {angularLimitsMax=<0 5 2>, angularLimitsMin=<0 -10 -2>, angularTarget=<1 0 0>}
      localJointOffset=<0.9 0 0>
     }]
    relativeSpaceBone {boneName=head}
    simSpaceSettings {maxAngularAcceleration=1500, maxAngularVelocity=1500, simSpaceAngularAlpha=0.2}
   }
   n7 = Anim Dynamics - Bone: dyn_curl_fr_03_l(ComponentPose=@n6, Alpha=1, ExternalForce="0,0,0") {
    angularDampingOverride=0.9
    angularSpringConstant=250
    bAngularSpring=true
    bGravityOverrideInSimSpace=true
    bOverrideAngularDamping=true
    bUseGravityOverride=true
    simulationSpace=BoneRelative
    boundBone {boneName=dyn_curl_fr_03_l}
    componentAppliedLinearAccClamp=<2000 2000 2000>
    componentLinearAccScale=<0.2 0.2 0.2>
    componentLinearVelScale=<0.2 0.2 0.2>
    gravityOverride {x=-980}
    physicsBodyDefinitions [{
      boundBone {boneName=dyn_curl_fr_03_l}
      boxExtents=<10 10 10>
      constraintSetup {angularLimitsMax=<0 5 2>, angularLimitsMin=<0 -10 -2>, angularTarget=<1 0 0>}
      localJointOffset=<0.8 0 0>
     }]
    relativeSpaceBone {boneName=head}
    simSpaceSettings {maxAngularAcceleration=1500, maxAngularVelocity=1500, simSpaceAngularAlpha=0.2}
   }
   n8 = Anim Dynamics - Bone: dyn_curl_fr_04_l(ComponentPose=@n7, Alpha=1, ExternalForce="0,0,0") {
    angularDampingOverride=0.9
    angularSpringConstant=250
    bAngularSpring=true
    bGravityOverrideInSimSpace=true
    bOverrideAngularDamping=true
    bUseGravityOverride=true
    simulationSpace=BoneRelative
    boundBone {boneName=dyn_curl_fr_04_l}
    componentAppliedLinearAccClamp=<2000 2000 2000>
    componentLinearAccScale=<0.2 0.2 0.2>
    componentLinearVelScale=<0.2 0.2 0.2>
    gravityOverride {x=-980}
    physicsBodyDefinitions [{
      boundBone {boneName=dyn_curl_fr_04_l}
      boxExtents=<9 9 9>
      constraintSetup {angularLimitsMax=<0 2 3>, angularLimitsMin=<0 -10 -3>, angularTarget=<1 0 0>}
      localJointOffset=<1 0 0>
     }]
    relativeSpaceBone {boneName=head}
    simSpaceSettings {maxAngularAcceleration=1500, maxAngularVelocity=1500, simSpaceAngularAlpha=0.2}
   }
   n9 = Anim Dynamics - Bone: dyn_curl_fr_05_l(ComponentPose=@n8, Alpha=1, ExternalForce="0,0,0") {
    angularDampingOverride=0.9
    angularSpringConstant=250
    bAngularSpring=true
    bGravityOverrideInSimSpace=true
    bOverrideAngularDamping=true
    bUseGravityOverride=true
    simulationSpace=BoneRelative
    boundBone {boneName=dyn_curl_fr_05_l}
    componentAppliedLinearAccClamp=<2000 2000 2000>
    componentLinearAccScale=<0.2 0.2 0.2>
    componentLinearVelScale=<0.2 0.2 0.2>
    gravityOverride {x=-980}
    physicsBodyDefinitions [{
      boundBone {boneName=dyn_curl_fr_05_l}
      boxExtents=<9 9 9>
      constraintSetup {angularLimitsMax=<0 5 2>, angularLimitsMin=<0 -5 -2>, angularTarget=<1 0 0>}
      localJointOffset=<1 0 0>
     }]
    relativeSpaceBone {boneName=head}
    simSpaceSettings {maxAngularAcceleration=1500, maxAngularVelocity=1500, simSpaceAngularAlpha=0.2}
   }
   n10 = Anim Dynamics - Bone: dyn_curl_fr_01_r(ComponentPose=@n9, Alpha=1, ExternalForce="0,0,0") {
    angularDampingOverride=0.9
    angularSpringConstant=250
    bAngularSpring=true
    bGravityOverrideInSimSpace=true
    bOverrideAngularDamping=true
    bUseGravityOverride=true
    simulationSpace=BoneRelative
    boundBone {boneName=dyn_curl_fr_01_r}
    componentAppliedLinearAccClamp=<2000 2000 2000>
    componentLinearAccScale=<0.2 0.2 0.2>
    componentLinearVelScale=<0.2 0.2 0.2>
    gravityOverride {x=-980}
    physicsBodyDefinitions [{
      boundBone {boneName=dyn_curl_fr_01_r}
      boxExtents=<11 11 11>
      constraintSetup {angularLimitsMax=<0 5 2>, angularLimitsMin=<0 -10 -2>, angularTarget=<1 0 0>}
      localJointOffset=<0.7 0 0>
     }]
    relativeSpaceBone {boneName=head}
    simSpaceSettings {maxAngularAcceleration=1500, maxAngularVelocity=1500, simSpaceAngularAlpha=0.2}
   }
   n11 = Anim Dynamics - Bone: dyn_curl_fr_02_r(ComponentPose=@n10, Alpha=1, ExternalForce="0,0,0") {
    angularDampingOverride=0.9
    angularSpringConstant=250
    bAngularSpring=true
    bGravityOverrideInSimSpace=true
    bOverrideAngularDamping=true
    bUseGravityOverride=true
    simulationSpace=BoneRelative
    boundBone {boneName=dyn_curl_fr_02_r}
    componentAppliedLinearAccClamp=<2000 2000 2000>
    componentLinearAccScale=<0.2 0.2 0.2>
    componentLinearVelScale=<0.2 0.2 0.2>
    gravityOverride {x=-980}
    physicsBodyDefinitions [{
      boundBone {boneName=dyn_curl_fr_02_r}
      boxExtents=<11 11 11>
      constraintSetup {angularLimitsMax=<0 5 2>, angularLimitsMin=<0 -5 -2>, angularTarget=<1 0 0>}
      localJointOffset=<0.7 0 0>
     }]
    relativeSpaceBone {boneName=head}
    simSpaceSettings {maxAngularAcceleration=1500, maxAngularVelocity=1500, simSpaceAngularAlpha=0.2}
   }
   n12 = Anim Dynamics - Bone: dyn_curl_fr_05_r(ComponentPose=@n11, Alpha=1, ExternalForce="0,0,0") {
    angularDampingOverride=0.9
    angularSpringConstant=250
    bAngularSpring=true
    bGravityOverrideInSimSpace=true
    bOverrideAngularDamping=true
    bUseGravityOverride=true
    simulationSpace=BoneRelative
    boundBone {boneName=dyn_curl_fr_05_r}
    componentAppliedLinearAccClamp=<2000 2000 2000>
    componentLinearAccScale=<0.2 0.2 0.2>
    componentLinearVelScale=<0.2 0.2 0.2>
    gravityOverride {x=-980}
    physicsBodyDefinitions [{
      boundBone {boneName=dyn_curl_fr_05_r}
      boxExtents=<11 11 11>
      constraintSetup {angularLimitsMax=<0 3 5>, angularLimitsMin=<0 -3 -5>, angularTarget=<1 0 0>}
      localJointOffset=<0.7 0 0>
     }]
    relativeSpaceBone {boneName=head}
    simSpaceSettings {maxAngularAcceleration=1500, maxAngularVelocity=1500, simSpaceAngularAlpha=0.2}
   }
   n13 = Anim Dynamics - Bone: dyn_curl_fr_06_r(ComponentPose=@n12, Alpha=1, ExternalForce="0,0,0") {
    angularDampingOverride=0.9
    angularSpringConstant=250
    bAngularSpring=true
    bGravityOverrideInSimSpace=true
    bOverrideAngularDamping=true
    bUseGravityOverride=true
    simulationSpace=BoneRelative
    boundBone {boneName=dyn_curl_fr_06_r}
    componentAppliedLinearAccClamp=<2000 2000 2000>
    componentLinearAccScale=<0.2 0.2 0.2>
    componentLinearVelScale=<0.2 0.2 0.2>
    gravityOverride {x=-980}
    physicsBodyDefinitions [{
      boundBone {boneName=dyn_curl_fr_06_r}
      boxExtents=<11 11 11>
      constraintSetup {angularLimitsMax=<0 10 5>, angularLimitsMin=<0 -5 -3>, angularTarget=<1 0 0>}
      localJointOffset=<0.7 0 0>
     }]
    relativeSpaceBone {boneName=head}
    simSpaceSettings {maxAngularAcceleration=1500, maxAngularVelocity=1500, simSpaceAngularAlpha=0.2}
   }
   n14 = Anim Dynamics - Bone: dyn_curl_fr_07_r(ComponentPose=@n13, Alpha=1, ExternalForce="0,0,0") {
    angularDampingOverride=0.9
    angularSpringConstant=250
    bAngularSpring=true
    bGravityOverrideInSimSpace=true
    bOverrideAngularDamping=true
    bUseGravityOverride=true
    simulationSpace=BoneRelative
    boundBone {boneName=dyn_curl_fr_07_r}
    componentAppliedLinearAccClamp=<2000 2000 2000>
    componentLinearAccScale=<0.2 0.2 0.2>
    componentLinearVelScale=<0.2 0.2 0.2>
    gravityOverride {x=-980}
    physicsBodyDefinitions [{
      boundBone {boneName=dyn_curl_fr_07_r}
      boxExtents=<11 11 11>
      constraintSetup {angularLimitsMax=<0 5 5>, angularLimitsMin=<0 -5 -5>, angularTarget=<1 0 0>}
      localJointOffset=<0.7 0 0>
     }]
    relativeSpaceBone {boneName=head}
    simSpaceSettings {maxAngularAcceleration=1500, maxAngularVelocity=1500, simSpaceAngularAlpha=0.2}
   }
   n15 = Anim Dynamics - Bone: dyn_curl_fr_08_r(ComponentPose=@n14, Alpha=1, ExternalForce="0,0,0") {
    angularDampingOverride=0.9
    angularSpringConstant=250
    bAngularSpring=true
    bGravityOverrideInSimSpace=true
    bOverrideAngularDamping=true
    bUseGravityOverride=true
    simulationSpace=BoneRelative
    boundBone {boneName=dyn_curl_fr_08_r}
    componentAppliedLinearAccClamp=<2000 2000 2000>
    componentLinearAccScale=<0.2 0.2 0.2>
    componentLinearVelScale=<0.2 0.2 0.2>
    gravityOverride {x=-980}
    physicsBodyDefinitions [{
      boundBone {boneName=dyn_curl_fr_08_r}
      boxExtents=<11 11 11>
      constraintSetup {angularLimitsMax=<0 10 10>, angularLimitsMin=<0 -5 -10>, angularTarget=<1 0 0>}
      localJointOffset=<0.7 0 0>
     }]
    relativeSpaceBone {boneName=head}
    simSpaceSettings {maxAngularAcceleration=1500, maxAngularVelocity=1500, simSpaceAngularAlpha=0.2}
   }
   n16 = Anim Dynamics - Bone: dyn_curl_fr_02_r(ComponentPose=@n15, Alpha=1, ExternalForce="0,0,0") {
    angularDampingOverride=0.9
    angularSpringConstant=250
    bAngularSpring=true
    bGravityOverrideInSimSpace=true
    bOverrideAngularDamping=true
    bUseGravityOverride=true
    simulationSpace=BoneRelative
    boundBone {boneName=dyn_curl_fr_02_r}
    componentAppliedLinearAccClamp=<2000 2000 2000>
    componentLinearAccScale=<0.2 0.2 0.2>
    componentLinearVelScale=<0.2 0.2 0.2>
    gravityOverride {x=-980}
    physicsBodyDefinitions [{
      boundBone {boneName=dyn_curl_fr_02_r}
      boxExtents=<11 11 11>
      constraintSetup {angularLimitsMax=<0 10 5>, angularLimitsMin=<0 -10 -5>, angularTarget=<1 0 0>}
      localJointOffset=<0.7 0 0>
     }]
    relativeSpaceBone {boneName=head}
    simSpaceSettings {maxAngularAcceleration=1500, maxAngularVelocity=1500, simSpaceAngularAlpha=0.2}
   }
   n17 = Anim Dynamics - Bone: dyn_curl_fr_03_r(ComponentPose=@n16, Alpha=1, ExternalForce="0,0,0") {
    angularDampingOverride=0.9
    angularSpringConstant=250
    bAngularSpring=true
    bGravityOverrideInSimSpace=true
    bOverrideAngularDamping=true
    bUseGravityOverride=true
    simulationSpace=BoneRelative
    boundBone {boneName=dyn_curl_fr_03_r}
    componentAppliedLinearAccClamp=<2000 2000 2000>
    componentLinearAccScale=<0.2 0.2 0.2>
    componentLinearVelScale=<0.2 0.2 0.2>
    gravityOverride {x=-980}
    physicsBodyDefinitions [{
      boundBone {boneName=dyn_curl_fr_03_r}
      boxExtents=<11 11 11>
      constraintSetup {angularLimitsMax=<0 10 5>, angularLimitsMin=<0 -10 -5>, angularTarget=<1 0 0>}
      localJointOffset=<0.7 0 0>
     }]
    relativeSpaceBone {boneName=head}
    simSpaceSettings {maxAngularAcceleration=1500, maxAngularVelocity=1500, simSpaceAngularAlpha=0.2}
   }
   n18 = Anim Dynamics - Bone: dyn_curl_fr_04_r(ComponentPose=@n17, Alpha=1, ExternalForce="0,0,0") {
    angularDampingOverride=0.9
    angularSpringConstant=250
    bAngularSpring=true
    bGravityOverrideInSimSpace=true
    bOverrideAngularDamping=true
    bUseGravityOverride=true
    simulationSpace=BoneRelative
    boundBone {boneName=dyn_curl_fr_04_r}
    componentAppliedLinearAccClamp=<2000 2000 2000>
    componentLinearAccScale=<0.2 0.2 0.2>
    componentLinearVelScale=<0.2 0.2 0.2>
    gravityOverride {x=-980}
    physicsBodyDefinitions [{
      boundBone {boneName=dyn_curl_fr_04_r}
      boxExtents=<11 11 11>
      constraintSetup {angularLimitsMax=<0 10 5>, angularLimitsMin=<0 -10 -5>, angularTarget=<1 0 0>}
      localJointOffset=<0.7 0 0>
     }]
    relativeSpaceBone {boneName=head}
    simSpaceSettings {maxAngularAcceleration=1500, maxAngularVelocity=1500, simSpaceAngularAlpha=0.2}
   }
   n19 = Anim Dynamics - Bone: dyn_curl_side_01_l(ComponentPose=@n18, Alpha=1, ExternalForce="0,0,0") {
    angularDampingOverride=0.9
    angularSpringConstant=250
    bAngularSpring=true
    bGravityOverrideInSimSpace=true
    bOverrideAngularDamping=true
    bUseGravityOverride=true
    simulationSpace=BoneRelative
    boundBone {boneName=dyn_curl_side_01_l}
    componentAppliedLinearAccClamp=<2000 2000 2000>
    componentLinearAccScale=<0.2 0.2 0.2>
    componentLinearVelScale=<0.2 0.2 0.2>
    gravityOverride {x=-980}
    physicsBodyDefinitions [{
      boundBone {boneName=dyn_curl_side_01_l}
      boxExtents=<11 11 11>
      constraintSetup {angularLimitsMax=<0 10 5>, angularLimitsMin=<0 -10 0>, angularTarget=<1 0 0>}
      localJointOffset=<0.7 0 0>
     }]
    relativeSpaceBone {boneName=head}
    simSpaceSettings {maxAngularAcceleration=1500, maxAngularVelocity=1500, simSpaceAngularAlpha=0.2}
   }
   n20 = Anim Dynamics - Bone: dyn_curl_side_02_l(ComponentPose=@n19, Alpha=1, ExternalForce="0,0,0") {
    angularDampingOverride=0.9
    angularSpringConstant=250
    bAngularSpring=true
    bGravityOverrideInSimSpace=true
    bOverrideAngularDamping=true
    bUseGravityOverride=true
    simulationSpace=BoneRelative
    boundBone {boneName=dyn_curl_side_02_l}
    componentAppliedLinearAccClamp=<2000 2000 2000>
    componentLinearAccScale=<0.2 0.2 0.2>
    componentLinearVelScale=<0.2 0.2 0.2>
    gravityOverride {x=-980}
    physicsBodyDefinitions [{
      boundBone {boneName=dyn_curl_side_02_l}
      boxExtents=<11 11 11>
      constraintSetup {angularLimitsMax=<0 10 0>, angularLimitsMin=<0 -10 0>, angularTarget=<1 0 0>}
      localJointOffset=<0.7 0 0>
     }]
    relativeSpaceBone {boneName=head}
    simSpaceSettings {maxAngularAcceleration=1500, maxAngularVelocity=1500, simSpaceAngularAlpha=0.2}
   }
   n21 = Anim Dynamics - Bone: dyn_curl_side_03_l(ComponentPose=@n20, Alpha=1, ExternalForce="0,0,0") {
    angularDampingOverride=0.9
    angularSpringConstant=250
    bAngularSpring=true
    bGravityOverrideInSimSpace=true
    bOverrideAngularDamping=true
    bUseGravityOverride=true
    simulationSpace=BoneRelative
    boundBone {boneName=dyn_curl_side_03_l}
    componentAppliedLinearAccClamp=<2000 2000 2000>
    componentLinearAccScale=<0.2 0.2 0.2>
    componentLinearVelScale=<0.2 0.2 0.2>
    gravityOverride {x=-980}
    physicsBodyDefinitions [{
      boundBone {boneName=dyn_curl_side_03_l}
      boxExtents=<11 11 11>
      constraintSetup {angularLimitsMax=<0 10 10>, angularLimitsMin=<0 -10 -15>, angularTarget=<1 0 0>}
      localJointOffset=<0.7 0 0>
     }]
    relativeSpaceBone {boneName=head}
    simSpaceSettings {maxAngularAcceleration=1500, maxAngularVelocity=1500, simSpaceAngularAlpha=0.2}
   }
   n22 = Anim Dynamics - Bone: def_ear_piercing(ComponentPose=@n21, Alpha=1, ExternalForce="0,0,0") {
    angularDampingOverride=0.9
    angularSpringConstant=300
    bAngularSpring=true
    bOverrideAngularDamping=true
    bUseGravityOverride=true
    simulationSpace=RootRelative
    boundBone {boneName=def_ear_piercing}
    componentAppliedLinearAccClamp=<3000 3000 3000>
    componentLinearAccScale=<0.1 0.1 0.1>
    componentLinearVelScale=<0.1 0.1 0.3>
    gravityOverride {z=-980}
    physicsBodyDefinitions [{
      boundBone {boneName=def_ear_piercing}
      boxExtents=<10 10 10>
      constraintSetup {angularLimitsMax=<0 0 30>, angularLimitsMin=<0 0 -30>, angularTarget=<0 0 0>}
      localJointOffset=<0 1 0>
     }]
    relativeSpaceBone {boneName=head}
    simSpaceSettings {maxAngularAcceleration=1500, maxAngularVelocity=1500, simSpaceAngularAlpha=0.1}
   }
   Output Pose(Result=Component To Local(ComponentPose=@n22)) {layerGroup=None}
  ]}
 EventGraph {nodes [
   Event BlueprintUpdateAnimation()
   > n1 = "Gravity Override State Function (Parameter Asset) (Nativized)"(
    CharacterPartAnimInstance=Self-Reference()
    GameTimeInSeconds=GetGameTimeInSeconds()
    Parameters=/BRCosmetics/Characters/Player/Female/Medium/Bodies/F_MED_PolkaSkate/Meshes/F_MED_PolkaSkate_Parameters
   )
   > Set Gravity Override(Gravity Override=@n1)
  ]}
]
variables [Gravity Override {type=Vector}]
